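# Behavior tree for the medium Tidal Crane swarm NPC, choosing between holding, repositioning and ranged attack.
# A root selector with a leash AI evaluator service first idles during spawn-in or despawn tags, then triggers Reposition when a threat is near.
# Otherwise, with a goal actor, it fires the ranged attack ability inside a narrow cone, falling back to waits.
# Actions run as tag-selected gameplay abilities, with observer aborts enabling priority interrupts.
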BehaviorTree /ProtoSwarm/NPCs/Swarm/TidalCrane_Swarm_Medium/BehaviorTree/BT_TidalCrane_Medium
blackboard=&/BB_TidalCrane_Medium
root {
 name=Selector
 class=BTComposite_Selector
 children [
  Selector {
   class=BTComposite_Selector
   children [Wait {class=BTTask_Wait, properties {waitTime {defaultValue=5}}}]
   decorators [Spawn In in progress {
     class=BTDecorator_GameplayTagQuery
     properties {
      flowAbortMode=Both
      nodeName=Spawn In in progress
      gameplayTagQuery {
       autoDescription=" ANY( Ability.TidalCrane.Lobber.SpawnIn, Ability.TidalCrane.Lobber.Despawn, Ability.TidalCrane.Despawn )"
       queryTokenStream [0,1,1,3,0,1,2]
       tagDictionary [{tagName=Ability.TidalCrane.Lobber.SpawnIn}, {tagName=Ability.TidalCrane.Lobber.Despawn}, {tagName=Ability.TidalCrane.Despawn}]
      }
     }
    }]
   properties {
    children [{childComposite=null, childTask=&self:BTTask_Wait_2, decorators []}]
    services []
   }
   services []
  }
  Reposition {
   class=FortBTTask_ExecuteGameplayAbility
   decorators [
    Is Threat In Proximity? {
     class=BTDecorator_Blackboard
     properties {
      arithmeticOperation=Equal
      flowAbortMode=LowerPriority
      intValue=0
      nodeName=Is Threat In Proximity?
      stringValue=""
      blackboardKey {selectedKeyName=AIEvaluator_ThreatInProximity}
     }
    }
    Gameplay Tag Query {
     class=BTDecorator_GameplayTagQuery
     properties {
      flowAbortMode=None
      nodeName=Gameplay Tag Query
      gameplayTagQuery {
       autoDescription=" NONE( Ability.TidalCrane.Lobber.SpawnIn, Ability.TidalCrane.Lobber.Despawn, Ability.TidalCrane.Despawn )"
       queryTokenStream [0,1,3,3,0,1,2]
       tagDictionary [{tagName=Ability.TidalCrane.Lobber.SpawnIn}, {tagName=Ability.TidalCrane.Lobber.Despawn}, {tagName=Ability.TidalCrane.Despawn}]
      }
     }
    }
    Has Gameplay Ability Reposition {
     class=FortBTDecorator_GameplayAbility_HasGameplayAbility
     properties {
      nodeName=Has Gameplay Ability Reposition
      gameplayAbilityTagContainer {defaultValue {gameplayTags [{tagName=Ability.ProtoSwarm.NonAttack.Reposition}]}}
     }
    }
    Can Activate Gameplay Ability Reposition {
     class=FortBTDecorator_GameplayAbility_CanActivate
     properties {
      flowAbortMode=LowerPriority
      nodeName=Can Activate Gameplay Ability Reposition
      gameplayAbilityTagContainer {defaultValue {gameplayTags [{tagName=Ability.ProtoSwarm.NonAttack.Reposition}]}}
      targetValue {key=None}
     }
    }
   ]
   properties {nodeName=Reposition, gameplayAbilityTagContainer {defaultValue {gameplayTags [{tagName=Ability.ProtoSwarm.NonAttack.Reposition}]}}}
  }
  Attack {
   class=BTComposite_Selector
   children [
    Range Attack {
     class=FortBTTask_ExecuteGameplayAbility
     decorators [
      Cone Check {
       class=BTDecorator_ConeCheck
       properties {
        bInverseCondition=false
        flowAbortMode=LowerPriority
        nodeName=Cone Check
        coneDirection {bNoneIsAllowedValue=true, selectedKeyName=None}
        coneHalfAngle {defaultValue=10, key=None}
        coneOrigin {bNoneIsAllowedValue=false, selectedKeyName=SelfActor}
        observed {bNoneIsAllowedValue=false, selectedKeyName=GoalActor}
       }
      }
      Can Activate Gameplay Ability RangeAttack {
       class=FortBTDecorator_GameplayAbility_CanActivate
       properties {
        flowAbortMode=None
        nodeName=Can Activate Gameplay Ability RangeAttack
        gameplayAbilityTagContainer {defaultValue {gameplayTags [{tagName=Ability.ProtoSwarm.Attack.Range}]}}
        targetValue {key=None}
       }
      }
      Has Gameplay Ability Range Attack {
       class=FortBTDecorator_GameplayAbility_HasGameplayAbility
       properties {
        nodeName=Has Gameplay Ability Range Attack
        gameplayAbilityTagContainer {defaultValue {gameplayTags [{tagName=Ability.ProtoSwarm.Attack.Range}]}}
       }
      }
     ]
     properties {nodeName=Range Attack, gameplayAbilityTagContainer {defaultValue {gameplayTags [{tagName=Ability.ProtoSwarm.Attack.Range}]}}}
    }
    Wait {class=BTTask_Wait, properties {waitTime {defaultValue=5}}}
   ]
   decorators [Has Goal Actor? {
     class=BTDecorator_Blackboard
     properties {
      arithmeticOperation=Equal
      intValue=0
      nodeName=Has Goal Actor?
      notifyObserver=ValueChange
      stringValue=""
      blackboardKey {selectedKeyName=GoalActor}
     }
    }]
   properties {
    nodeName=Attack
    children [
     {
      childComposite=null
      childTask=&self:FortBTTask_ExecuteGameplayAbility_2
      decorators [
       &self:BTDecorator_ConeCheck_0
       &self:FortBTDecorator_GameplayAbility_CanActivate_0
       &self:FortBTDecorator_GameplayAbility_HasGameplayAbility_0
      ]
     }
     {childComposite=null, childTask=&self:BTTask_Wait_1, decorators []}
    ]
    services []
   }
   services []
  }
  Wait {class=BTTask_Wait, properties {waitTime {defaultValue=5}}}
 ]
 decorators []
 properties {
  children [
   {
    childComposite=&self:BTComposite_Selector_1
    childTask=null
    decorators [&self:BTDecorator_GameplayTagQuery_0]
   }
   {
    childComposite=null
    childTask=&self:FortBTTask_ExecuteGameplayAbility_4
    decorators [
     &self:BTDecorator_Blackboard_2
     &self:BTDecorator_GameplayTagQuery_1
     &self:FortBTDecorator_GameplayAbility_HasGameplayAbility_2
     &self:FortBTDecorator_GameplayAbility_CanActivate_3
    ]
   }
   {
    childComposite=&self:BTComposite_Selector_48
    childTask=null
    decorators [&self:BTDecorator_Blackboard_0]
   }
   {childComposite=null, childTask=&self:BTTask_Wait_0, decorators []}
  ]
  services [&self:FortAthenaBTService_AIEvaluator_0]
 }
 services ["AI Evaluator: Leash" {
   class=FortAthenaBTService_AIEvaluator
   properties {interval=0.5, nodeName=AI Evaluator: Leash, aIEvaluatorClass {defaultValue=/Script/FortniteAIServer.FortAthenaAIEvaluator_Leash}}
  }]
}
